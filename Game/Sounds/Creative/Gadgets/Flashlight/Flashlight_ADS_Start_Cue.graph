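# Sound cue for the flashlight gadget's aim-down-sights start, built from reused foley and weapon sounds.
# A mixer layers a quieter random pick of two player run foley clips with a shotgun ADS-in sound pitched up by a fixed modulator.
# A root modulator wraps the mix, which plays at reduced volume in the weapon foley class with small flashlight attenuation and a concurrency override that prevents new instances at the limit.
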SoundCue /Game/Sounds/Creative/Gadgets/Flashlight/Flashlight_ADS_Start_Cue
root=Modulator_0
nodes [
 {id=Mixer_0, children [Random_4,Modulator_2], properties {inputVolume [0.3,1]}}
 {id=Modulator_0, children [Mixer_0]}
 {id=Modulator_2, children [WavePlayer_0], properties {pitchMax=1.3, pitchMin=1.3, volumeMax=1, volumeMin=1}}
 {id=Random_4, children [WavePlayer_1563, WavePlayer_1562], properties {weights [1,1]}}
 {id=WavePlayer_0, properties {soundWaveAssetPtr=/Game/Sounds/Fort_Weapons/Shotgun/CombatNew/Shotgun_CombatNew_ADS_In_01}}
 {id=WavePlayer_1562, properties {soundWaveAssetPtr=/Game/Sounds/Fort_Foley_Sounds/Moves/player_run_foley_12}}
 {id=WavePlayer_1563, properties {soundWaveAssetPtr=/Game/Sounds/Fort_Foley_Sounds/Moves/player_run_foley_10}}
]
properties {
 attenuationSettings=/Game/Sounds/Attenuation/Fort_Weapon_Flashlight_Sml_Att
 bOverrideConcurrency=true
 maxDistance=1000
 soundClassObject=&Fort_WeaponFoley
 volumeMultiplier=0.226
 attenuationOverrides {falloffDistance=1500, nonSpatializedRadiusStart=200, attenuationShapeExtents {x=1200}}
 concurrencyOverrides {resolutionRule=PreventNew}
}
