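# Sound cue for the Creative flashlight gadget, named as the sound for leaving aim-down-sights.
# A Mixer layers a quiet running foley clip, cut short by an Enveloper volume fade, under a louder reused combat shotgun ADS-out sample.
# The mix passes through a default Modulator for variation and plays at low overall volume in the weapon foley class with small flashlight attenuation.
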SoundCue /Game/Sounds/Creative/Gadgets/Flashlight/Flashlight_ADS_Stop_Cue
root=Modulator_0
nodes [
 {
  id=Enveloper_0
  children [WavePlayer_7]
  properties {
   pitchCurve {editorCurveData {keys [{time=0}]}}
   volumeCurve {editorCurveData {keys [<at 0 1>, {arriveTangent=-2.8571, interpMode=RCIM_Cubic, leaveTangent=-2.8571, tangentMode=RCTM_Break, time=0.2, value=1}, <at 0.27 0>]}}
  }
 }
 {id=Mixer_0, children [Enveloper_0,WavePlayer_0], properties {inputVolume [0.25,1.15]}}
 {id=Modulator_0, children [Mixer_0]}
 {id=WavePlayer_0, properties {soundWaveAssetPtr=/Game/Sounds/Fort_Weapons/Shotgun/CombatNew/Shotgun_CombatNew_ADS_Out_01}}
 {id=WavePlayer_7, properties {soundWaveAssetPtr=/Game/Sounds/Fort_Foley_Sounds/Moves/player_run_foley_01}}
]
properties {
 attenuationSettings=/Game/Sounds/Attenuation/Fort_Weapon_Flashlight_Sml_Att
 maxDistance=1000
 soundClassObject=&Fort_WeaponFoley
 volumeMultiplier=0.2
}
