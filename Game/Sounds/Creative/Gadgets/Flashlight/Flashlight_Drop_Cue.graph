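# Sound cue for the drop sound of the Creative Flashlight gadget.
# A single wave player reuses the Medieval Pistol drop wave, pitched up and lowered in volume at the cue level.
# It uses the shared weapon drop attenuation, the loot collect sound class and the Weapon_Drop_StopOldest concurrency set.
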SoundCue /Game/Sounds/Creative/Gadgets/Flashlight/Flashlight_Drop_Cue
root=WavePlayer_2
nodes [{id=WavePlayer_2, properties {soundWaveAssetPtr=/Game/Sounds/Fort_Weapons/Medieval/Pistol/Medieval_Pistol_Drop_01}}]
properties {
 attenuationSettings=&Fort_Weapon_Drop_Att
 maxDistance=3000
 pitchMultiplier=1.5
 soundClassObject=&Fort_Loot_Collect
 volumeMultiplier=0.4
 concurrencyOverrides {resolutionRule=PreventNew}
 concurrencySet [&Weapon_Drop_StopOldest]
}
