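# Sound cue for equipping the Creative flashlight gadget, built from reused foley and weapon sounds.
# A Mixer layers immediate player run foley with two briefly delayed clicks: a shotgun load handle, then a half-level medieval pistol reload end.
# A root Modulator feeds the result out through the player foley class with small flashlight attenuation.
# Its concurrency override blocks new instances rather than stopping playing ones.
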SoundCue /Game/Sounds/Creative/Gadgets/Flashlight/Flashlight_Equip_Cue
root=Modulator_0
nodes [
 {id=Delay_1, children [WavePlayer_0], properties {delayMax=0.15, delayMin=0.15}}
 {id=Delay_2, children [WavePlayer_1544], properties {delayMax=0.08, delayMin=0.08}}
 {id=Mixer_0, children [WavePlayer_1537, Delay_1, Delay_2], properties {inputVolume [1,0.5,1]}}
 {id=Modulator_0, children [Mixer_0]}
 {id=WavePlayer_0, properties {soundWaveAssetPtr=/Game/Sounds/Fort_Weapons/Medieval/Pistol/Medieval_Pistol_ReloadEnd_01}}
 {id=WavePlayer_1537, properties {soundWaveAssetPtr=/Game/Sounds/Fort_Foley_Sounds/Moves/player_run_foley_01}}
 {id=WavePlayer_1544, properties {soundWaveAssetPtr=/Game/Sounds/Fort_Weapons/Shotgun/shotgun_load_handle_04}}
]
properties {
 attenuationSettings=/Game/Sounds/Attenuation/Fort_Weapon_Flashlight_Sml_Att
 bOverrideConcurrency=true
 maxDistance=1000
 soundClassObject=&Fort_Player_Foley
 volumeMultiplier=0.5
 attenuationOverrides {nonSpatializedRadiusStart=200}
 concurrencyOverrides {resolutionRule=PreventNew}
}
