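# Sound cue for the looping hum of the Creative flashlight gadget, routed to the gadgets items sound class at reduced volume.
# A modulator applies slight random pitch variation to a looping wave player of the flashlight loop.
# Overridden attenuation gives a short falloff range with listener focus, so the sound is quieter when out of focus or occluded.
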SoundCue /Game/Sounds/Creative/Gadgets/Flashlight/Flashlight_Loop_Cue
root=Modulator_0
nodes [
 {id=Modulator_0, children [WavePlayer_1], properties {pitchMax=1.02, pitchMin=0.98, volumeMax=1, volumeMin=1}}
 {id=WavePlayer_1, properties {bLooping=true, soundWaveAssetPtr=&/Flashlight_Loop01}}
]
properties {
 bOverrideAttenuation=true
 maxDistance=600
 soundClassObject=&Fort_Gadgets_Items
 volumeMultiplier=0.25
 attenuationOverrides {
  bEnableFocusInterpolation=true
  bEnableListenerFocus=true
  falloffDistance=600
  nonFocusVolumeAttenuation=0.3
  occlusionVolumeAttenuation=0.3
  stereoSpread=0
  attenuationShapeExtents {x=0}
 }
}
