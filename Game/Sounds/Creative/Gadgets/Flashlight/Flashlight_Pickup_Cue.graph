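# Sound cue for picking up the Creative flashlight gadget.
# A single wave player reuses the machine pistol pickup sound, raised in pitch and lowered in volume.
# It uses weapon pickup attenuation and the loot collect sound class, with a concurrency override that allows one instance and blocks new plays while it sounds.
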SoundCue /Game/Sounds/Creative/Gadgets/Flashlight/Flashlight_Pickup_Cue
root=WavePlayer_0
nodes [{id=WavePlayer_0, properties {soundWaveAssetPtr=/Game/Sounds/Fort_Weapons/pistol/MachinePistol/SoundWaves/Machine_Pistol_Pickup_01}}]
properties {
 attenuationSettings=&Fort_Weapon_Pickup_Att
 bOverrideConcurrency=true
 maxDistance=3500
 pitchMultiplier=1.2
 soundClassObject=&Fort_Loot_Collect
 volumeMultiplier=0.358
 concurrencyOverrides {maxCount=1, resolutionRule=PreventNew}
}
