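# Sound cue for equipping the pistol flashlight gadget, routed to the Fort_Gadgets_Items sound class.
# A single Flashlight_Start01 wave feeds a Modulator at default settings for slight random pitch and volume variation, then a short fixed Delay at the root.
# The cue plays the wave pitched up and quieter, with its own attenuation override for distance falloff instead of a shared attenuation asset.
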SoundCue /Game/Sounds/Creative/Gadgets/Flashlight/Flashlight_Pistol_Equip_Cue
root=Delay_0
nodes [
 {id=Delay_0, children [Modulator_0], properties {delayMax=0.04, delayMin=0.04}}
 {id=Modulator_0, children [WavePlayer_0]}
 {id=WavePlayer_0, properties {soundWaveAssetPtr=&/Flashlight_Start01}}
]
properties {
 bOverrideAttenuation=true
 maxDistance=1000
 pitchMultiplier=1.3
 soundClassObject=&Fort_Gadgets_Items
 volumeMultiplier=0.3
 attenuationOverrides {
  dBAttenuationAtMax=-40
  falloffDistance=800
  occlusionLowPassFilterFrequency=10000
  occlusionVolumeAttenuation=0.3
  stereoSpread=0
  attenuationShapeExtents {x=200}
 }
}
