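# Sound cue for the start sound of the Creative flashlight gadget.
# A single wave player feeds Flashlight_Start01 through a modulator left at default settings, with the cue volume lowered.
# It routes to the gadgets items sound class and takes spatialization from a small weapon flashlight attenuation asset.
# Stored attenuation overrides include occlusion, but the override switch is not set, so the shared asset applies.
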SoundCue /Game/Sounds/Creative/Gadgets/Flashlight/Flashlight_Start_Cue
root=Modulator_0
nodes [
 {id=Modulator_0, children [WavePlayer_0]}
 {id=WavePlayer_0, properties {soundWaveAssetPtr=&/Flashlight_Start01}}
]
properties {
 attenuationSettings=/Game/Sounds/Attenuation/Fort_Weapon_Flashlight_Sml_Att
 maxDistance=1000
 soundClassObject=&Fort_Gadgets_Items
 volumeMultiplier=0.4
 attenuationOverrides {
  bEnableOcclusion=true
  dBAttenuationAtMax=-40
  falloffDistance=1024
  occlusionLowPassFilterFrequency=10000
  occlusionTraceChannel=ECC_GameTraceChannel11
  occlusionVolumeAttenuation=0.2
  stereoSpread=0
  attenuationShapeExtents {x=200}
 }
}
